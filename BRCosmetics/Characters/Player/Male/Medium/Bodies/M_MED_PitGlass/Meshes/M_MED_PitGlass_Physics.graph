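# Physics asset for the M_MED_PitGlass male medium body cosmetic, defining secondary motion for a hanging strap.
# Four small capsule bodies, dyn_strap_00 to 03, form a constrained chain rooted at base_strap_00.
# Each joint has tight twist and asymmetric swing limits, with the root joint swinging wider on one axis.
# A large unconstrained capsule on thigh_r provides a leg collision shape alongside the strap.
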PhysicsAsset /BRCosmetics/Characters/Player/Male/Medium/Bodies/M_MED_PitGlass/Meshes/M_MED_PitGlass_Physics
bodies {
 base_strap_00 [{shapeType=Capsule, length=0.505, radius=0.505, shapeName=base_strap_00_capsule, rotation=<rot 0 -90 0>}]
 dyn_strap_00 [{shapeType=Capsule, length=2, radius=1.3, shapeName=dyn_strap_00_capsule, center=<2 0 0>, rotation=<rot 90 0 0>}]
 dyn_strap_01 [{shapeType=Capsule, length=2, radius=1.3, shapeName=dyn_strap_01_capsule, center=<2 0 0>, rotation=<rot 90 0 0>}]
 dyn_strap_02 [{shapeType=Capsule, length=2, radius=1.3, shapeName=dyn_strap_02_capsule, center=<2 0 0>, rotation=<rot 90 0 0>}]
 dyn_strap_03 [{shapeType=Capsule, length=2, radius=1.3, shapeName=dyn_strap_03_capsule, center=<2 0 0>, rotation=<rot 90 0 0>}]
 thigh_r [{shapeType=Capsule, length=50, radius=8, shapeName=thigh_r_capsule, center=<5.1491 0.4406 -0.5387>, rotation=<rot 90 180 -177.5869>}]
}
constraints [
 {bone1Name=dyn_strap_00, bone2Name=base_strap_00, swing1LimitDegrees=15, swing2LimitDegrees=40, twistLimitDegrees=2}
 {bone1Name=dyn_strap_01, bone2Name=dyn_strap_00, swing1LimitDegrees=15, swing2LimitDegrees=30, twistLimitDegrees=2}
 {bone1Name=dyn_strap_02, bone2Name=dyn_strap_01, swing1LimitDegrees=15, swing2LimitDegrees=30, twistLimitDegrees=2}
 {bone1Name=dyn_strap_03, bone2Name=dyn_strap_02, swing1LimitDegrees=15, swing2LimitDegrees=30, twistLimitDegrees=2}
]
