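# Post-process-style cosmetic AnimBlueprint for the PitGlass male body, animating gear, sleeves and floating shards atop the parent mesh's copied pose.
# Cached stages chain weighted surface constraints, sine-driven shard bobbing, sleeve anti-clipping, a deform Control Rig, AnimDynamics secondary motion and a final pelvis-space RigidBody.
# EventGraph turns bone deltas and targeting, skydiving and parachute states into corrective bone alphas or rotations, and updates RigidBody gravity.
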AnimBlueprint /BRCosmetics/Characters/Player/Male/Medium/Bodies/M_MED_PitGlass/Meshes/M_MED_PitGlass_AnimBP
target_skeleton=&/M_MED_PitGlass_Skeleton
graphs [
 AnimGraph {nodes [
   n1 = "Transform (Modify) Bone - Bone: base_chest_bag_r"(
    ComponentPose=Local To Component(LocalPose=Use cached pose 'animDynCache'())
    Alpha=Get avoid_clippingBracelete_targeting() -> (avoid_clippingBracelete_targeting:double)
    Translation="-3.592,-3.854,-1.658"
    Rotation="9.845,-1.749,-10.107"
    Scale="1,1,1"
   ) {
    rotationMode=BMM_Additive
    rotationSpace=BCS_BoneSpace
    translationMode=BMM_Additive
    translationSpace=BCS_BoneSpace
    alphaBoolBlend {blendOption=Linear}
    alphaScaleBiasClamp {bInterpResult=true, interpSpeedDecreasing=0.2, interpSpeedIncreasing=0}
    boneToModify {boneName=base_chest_bag_r}
    rotation=<rot 9.845 -1.749 -10.107>
    translation=<-3.592 -3.854 -1.658>
   }
   n2 = "Transform (Modify) Bone - Bone: base_strap_00"(ComponentPose=@n1, Alpha=Get thigh_rot_up_r() -> (thigh_rot_up_r:double), Translation="0,-5,6", Rotation="40,0,-30", Scale="1,1,1") {
    rotationMode=BMM_Additive
    rotationSpace=BCS_BoneSpace
    translationMode=BMM_Additive
    translationSpace=BCS_BoneSpace
    alphaBoolBlend {blendOption=Linear}
    boneToModify {boneName=base_strap_00}
    rotation {pitch=40, roll=-30}
    translation {y=-5, z=6}
   }
   n3 = "Transform (Modify) Bone - Bone: base_knee_brace"(ComponentPose=@n2, Alpha=Get Knee_Brace() -> (Knee_Brace:double), Translation="0,0,0.5", Rotation="70,0,0", Scale="1,1,1") {
    rotationMode=BMM_Additive
    rotationSpace=BCS_BoneSpace
    translationMode=BMM_Additive
    translationSpace=BCS_BoneSpace
    alphaBoolBlend {blendOption=Linear}
    boneToModify {boneName=base_knee_brace}
    rotation {pitch=70}
    translation {z=0.5}
   }
   n4 = RigidBody!(ComponentPose=@n3, Alpha, OverrideWorldGravity:Vector=Get GravityOverride(), ExternalForce) -> (Pose) {
    bOverrideWorldGravity=true
    overridePhysicsAsset=&/M_MED_PitGlass_Physics
    simulationSpace=BaseBoneSpace
    alphaBoolBlend {blendOption=Linear}
    baseBoneRef {boneName=pelvis}
    componentAppliedLinearAccClamp=<3000 3000 3000>
    componentLinearAccScale=<0.4 0.2 0.2>
    componentLinearVelScale=<0.4 0.2 0.2>
    simSpaceSettings {
     maxAngularAcceleration=1000
     maxAngularVelocity=3
     maxLinearAcceleration=700
     maxLinearVelocity=500
     velocityScaleZ=0
     worldAlpha=0.4
     externalLinearDragV=<0.2 0.2>
    }
   }
   Output Pose(Result=Component To Local(ComponentPose=@n4)) {layerGroup=None}
   n5 = "Constraint - base_gun_0_l (PPPPPPPP)"!(
    ComponentPose=Local To Component(LocalPose=Copy Pose From Mesh() {bUseAttachedParent=true})
    Alpha=1
    ConstraintWeights_0=0.3918
    ConstraintWeights_1:float=0.2775
    ConstraintWeights_2:float=0.1935
    ConstraintWeights_3:float=0.0803
    ConstraintWeights_4:float=0.0211
    ConstraintWeights_5:float=0.0195
    ConstraintWeights_6:float=0.0142
    ConstraintWeights_7:float=0.0019
   ) -> (Pose) {
    alphaBoolBlend {blendOption=Linear}
    boneToModify {boneName=base_gun_0_l}
    constraintSetup [
     {targetBone {boneName=deform_upperarm_up_l}}
     {targetBone {boneName=upperarm_twist_01_l}}
     {targetBone {boneName=clavicle_l}}
     {targetBone {boneName=upperarm_twist_02_l}}
     {targetBone {boneName=upperarm_l}}
     {targetBone {boneName=spine_05}}
     {targetBone {boneName=spine_03}}
     {targetBone {boneName=deform_upperarm_bk_l}}
    ]
    constraintWeights [1,1,1,1,1,1,1,1]
   }
   n6 = "Constraint - base_gun_0_r (PPPPPPPP)"!(
    ComponentPose=@n5
    Alpha=1
    ConstraintWeights_0=0.3918
    ConstraintWeights_1:float=0.2775
    ConstraintWeights_2:float=0.1935
    ConstraintWeights_3:float=0.0803
    ConstraintWeights_4:float=0.0211
    ConstraintWeights_5:float=0.0195
    ConstraintWeights_6:float=0.0142
    ConstraintWeights_7:float=0.0019
   ) -> (Pose) {
    alphaBoolBlend {blendOption=Linear}
    boneToModify {boneName=base_gun_0_r}
    constraintSetup [
     {targetBone {boneName=deform_upperarm_up_r}}
     {targetBone {boneName=upperarm_twist_01_r}}
     {targetBone {boneName=clavicle_r}}
     {targetBone {boneName=upperarm_twist_02_r}}
     {targetBone {boneName=upperarm_r}}
     {targetBone {boneName=spine_05}}
     {targetBone {boneName=spine_03}}
     {targetBone {boneName=deform_upperarm_bk_r}}
    ]
    constraintWeights [1,1,1,1,1,1,1,1]
   }
   n7 = "Constraint - base_gun_1_l (PPPPPPP)"!(
    ComponentPose=@n6
    Alpha=1
    ConstraintWeights_0=0.3172
    ConstraintWeights_1:float=0.2344
    ConstraintWeights_2:float=0.2168
    ConstraintWeights_3:float=0.1135
    ConstraintWeights_4:float=0.0732
    ConstraintWeights_5:float=0.0247
    ConstraintWeights_6:float=0.0202
   ) -> (Pose) {
    alphaBoolBlend {blendOption=Linear}
    boneToModify {boneName=base_gun_1_l}
    constraintSetup [
     {targetBone {boneName=spine_03}}
     {targetBone {boneName=spine_02}}
     {targetBone {boneName=spine_04}}
     {targetBone {boneName=spine_01}}
     {targetBone {boneName=pelvis}}
     {targetBone {boneName=clavicle_l}}
     {targetBone {boneName=spine_05}}
    ]
    constraintWeights [1,1,1,1,1,1,1]
   }
   n8 = "Constraint - base_gun_1_r (PPPPPPP)"!(
    ComponentPose=@n7
    Alpha=1
    ConstraintWeights_0=0.3172
    ConstraintWeights_1:float=0.2344
    ConstraintWeights_2:float=0.2168
    ConstraintWeights_3:float=0.1135
    ConstraintWeights_4:float=0.0732
    ConstraintWeights_5:float=0.0247
    ConstraintWeights_6:float=0.0202
   ) -> (Pose) {
    alphaBoolBlend {blendOption=Linear}
    boneToModify {boneName=base_gun_1_r}
    constraintSetup [
     {targetBone {boneName=spine_03}}
     {targetBone {boneName=spine_02}}
     {targetBone {boneName=spine_04}}
     {targetBone {boneName=spine_01}}
     {targetBone {boneName=pelvis}}
     {targetBone {boneName=clavicle_r}}
     {targetBone {boneName=spine_05}}
    ]
    constraintWeights [1,1,1,1,1,1,1]
   }
   n9 = "Constraint - base_knife (PPPPP)"!(
    ComponentPose=@n8
    Alpha=1
    ConstraintWeights_0=0.8055
    ConstraintWeights_1:float=0.0888
    ConstraintWeights_2:float=0.0599
    ConstraintWeights_3:float=0.0418
    ConstraintWeights_4:float=0.0041
   ) -> (Pose) {
    alphaBoolBlend {blendOption=Linear}
    boneToModify {boneName=base_knife}
    constraintSetup [
     {targetBone {boneName=thigh_twist_01_l}}
     {targetBone {boneName=thigh_l}}
     {targetBone {boneName=pelvis}}
     {targetBone {boneName=deform_groin_l}}
     {targetBone {boneName=deform_glute_l}}
    ]
    constraintWeights [1,1,1,1,1]
   }
   n10 = "Constraint - base_cannisters (PPPPP)"!(
    ComponentPose=@n9
    Alpha=1
    ConstraintWeights_0=0.6018
    ConstraintWeights_1:float=0.3626
    ConstraintWeights_2:float=0.0176
    ConstraintWeights_3:float=0.0116
    ConstraintWeights_4:float=0.0064
   ) -> (Pose) {
    alphaBoolBlend {blendOption=Linear}
    boneToModify {boneName=base_cannisters}
    constraintSetup [
     {targetBone {boneName=calf_twist_02_r}}
     {targetBone {boneName=calf_r}}
     {targetBone {boneName=calf_twist_01_r}}
     {targetBone {boneName=deform_knee_out_r}}
     {targetBone {boneName=thigh_r}}
    ]
    constraintWeights [1,1,1,1,1]
   }
   n11 = "Constraint - base_pantspull_l (PPP)"!(ComponentPose=@n10, Alpha=1, ConstraintWeights_0=0.5447, ConstraintWeights_1:float=0.434, ConstraintWeights_2:float=0.0213) -> (Pose) {
    alphaBoolBlend {blendOption=Linear}
    boneToModify {boneName=base_pantspull_l}
    constraintSetup [{targetBone {boneName=calf_twist_01_l}}, {targetBone {boneName=calf_twist_02_l}}, {targetBone {boneName=calf_l}}]
    constraintWeights [1,1,1]
   }
   n12 = "Constraint - base_strap_00 (PPPP)"!(
    ComponentPose=@n11
    Alpha=1
    ConstraintWeights_0=0.5451
    ConstraintWeights_1:float=0.4453
    ConstraintWeights_2:float=0.0094
    ConstraintWeights_3:float=0.0002
   ) -> (Pose) {
    alphaBoolBlend {blendOption=Linear}
    boneToModify {boneName=base_strap_00}
    constraintSetup [
     {targetBone {boneName=pelvis}}
     {targetBone {boneName=deform_groin_r}}
     {targetBone {boneName=thigh_twist_01_r}}
     {targetBone {boneName=thigh_r}}
    ]
    constraintWeights [1,1,1,1]
   }
   n13 = "Constraint - base_zipper_00_l (PPPP)"!(
    ComponentPose=@n12
    Alpha=1
    ConstraintWeights_0=0.76
    ConstraintWeights_1:float=0.1811
    ConstraintWeights_2:float=0.0295
    ConstraintWeights_3:float=0.0293
   ) -> (Pose) {
    alphaBoolBlend {blendOption=Linear}
    boneToModify {boneName=base_zipper_00_l}
    constraintSetup [{targetBone {boneName=clavicle_l}}, {targetBone {boneName=spine_05}}, {targetBone {boneName=spine_04}}, {targetBone {boneName=spine_03}}]
    constraintWeights [1,1,1,1]
   }
   n14 = "Constraint - base_zipper_00_r (PPPP)"!(
    ComponentPose=@n13
    Alpha=1
    ConstraintWeights_0=0.76
    ConstraintWeights_1:float=0.1811
    ConstraintWeights_2:float=0.0295
    ConstraintWeights_3:float=0.0293
   ) -> (Pose) {
    alphaBoolBlend {blendOption=Linear}
    boneToModify {boneName=base_zipper_00_r}
    constraintSetup [{targetBone {boneName=clavicle_r}}, {targetBone {boneName=spine_05}}, {targetBone {boneName=spine_04}}, {targetBone {boneName=spine_03}}]
    constraintWeights [1,1,1,1]
   }
   n15 = "Constraint - base_chest_bag_l (PPPPPPPP)"!(
    ComponentPose=@n14
    Alpha=1
    ConstraintWeights_0=0.4567
    ConstraintWeights_1:float=0.1745
    ConstraintWeights_2:float=0.1328
    ConstraintWeights_3:float=0.0982
    ConstraintWeights_4:float=0.0865
    ConstraintWeights_5:float=0.0437
    ConstraintWeights_6:float=0.0066
    ConstraintWeights_7:float=0.0008
   ) -> (Pose) {
    alphaBoolBlend {blendOption=Linear}
    boneToModify {boneName=base_chest_bag_l}
    constraintSetup [
     {targetBone {boneName=deform_pec_l}}
     {targetBone {boneName=spine_04}}
     {targetBone {boneName=clavicle_l}}
     {targetBone {boneName=spine_03}}
     {targetBone {boneName=spine_05}}
     {targetBone {boneName=spine_02}}
     {targetBone {boneName=spine_01}}
     {targetBone {boneName=pelvis}}
    ]
    constraintWeights [1,1,1,1,1,1,1,1]
   }
   n16 = "Constraint - base_chest_bag_r (PPPPPPPP)"!(
    ComponentPose=@n15
    Alpha=1
    ConstraintWeights_0=0.4548
    ConstraintWeights_1:float=0.1746
    ConstraintWeights_2:float=0.1343
    ConstraintWeights_3:float=0.0982
    ConstraintWeights_4:float=0.087
    ConstraintWeights_5:float=0.0437
    ConstraintWeights_6:float=0.0066
    ConstraintWeights_7:float=0.0008
   ) -> (Pose) {
    alphaBoolBlend {blendOption=Linear}
    boneToModify {boneName=base_chest_bag_r}
    constraintSetup [
     {targetBone {boneName=deform_pec_r}}
     {targetBone {boneName=spine_04}}
     {targetBone {boneName=clavicle_r}}
     {targetBone {boneName=spine_03}}
     {targetBone {boneName=spine_05}}
     {targetBone {boneName=spine_02}}
     {targetBone {boneName=spine_01}}
     {targetBone {boneName=pelvis}}
    ]
    constraintWeights [1,1,1,1,1,1,1,1]
   }
   Save cached pose 'SurfaceConstraintsCache'(Pose=Component To Local(ComponentPose=@n16)) {cacheName=SurfaceConstraintsCache}
   n17 = "Transform (Modify) Bone - Bone: dyn_backBracelete_00_r"(
    ComponentPose=Local To Component(LocalPose=Use cached pose 'animDyn_Guns_Cache'())
    Alpha=Get avoid_clippingBracelete_targeting()
    Translation="0,0,0"
    Rotation="-3.634,4.206,-3.932"
    Scale="1,1,1"
   ) {
    rotationMode=BMM_Additive
    rotationSpace=BCS_BoneSpace
    alphaBoolBlend {blendOption=Linear}
    alphaScaleBiasClamp {bInterpResult=true, interpSpeedDecreasing=0.2, interpSpeedIncreasing=0}
    boneToModify {boneName=dyn_backBracelete_00_r}
    rotation=<rot -3.634 4.206 -3.932>
   }
   n18 = Anim Dynamics - Bone: dyn_backBracelete_00_l(ComponentPose=@n17, Alpha=1, ExternalForce="0,0,0") {
    angularDampingOverride=1
    angularSpringConstant=10000
    bAngularSpring=true
    bOverrideAngularDamping=true
    bUseGravityOverride=true
    gravityScale=0.35
    simulationSpace=BoneRelative
    alphaBoolBlend {blendOption=Linear}
    boundBone {boneName=dyn_backBracelete_00_l}
    componentAppliedLinearAccClamp=<3000 3000 3000>
    componentLinearAccScale=<0.5 0.5 0.5>
    componentLinearVelScale=<0.5 0.5 0.5>
    gravityOverride {z=-980}
    physicsBodyDefinitions [{
      sphereCollisionRadius=0
      boundBone {boneName=dyn_backBracelete_00_l}
      boxExtents=<25 25 25>
      constraintSetup {angularLimitsMax=<0 0 5>, angularLimitsMin=<0 0 -5>, angularTarget=<1 0 0>}
      localJointOffset=<4 0 0>
     }]
    relativeSpaceBone {boneName=root}
   }
   n19 = Anim Dynamics - Bone: dyn_frontBracelete_00_l(ComponentPose=@n18, Alpha=1, ExternalForce="0,0,0") {
    angularDampingOverride=1
    angularSpringConstant=15000
    bAngularSpring=true
    bOverrideAngularDamping=true
    bUseGravityOverride=true
    gravityScale=0.35
    simulationSpace=BoneRelative
    alphaBoolBlend {blendOption=Linear}
    boundBone {boneName=dyn_frontBracelete_00_l}
    componentAppliedLinearAccClamp=<3000 3000 3000>
    componentLinearAccScale=<0.4 0.4 0.4>
    componentLinearVelScale=<0.4 0.4 0.4>
    gravityOverride {z=-980}
    physicsBodyDefinitions [{
      boundBone {boneName=dyn_frontBracelete_00_l}
      boxExtents=<28 28 28>
      constraintSetup {angularLimitsMax=<0 0 3>, angularLimitsMin=<0 0 -4>, angularTarget=<1 0 0>}
      localJointOffset=<4 0 0>
     }]
    relativeSpaceBone {boneName=root}
   }
   n20 = Anim Dynamics - Bone: dyn_frontBracelete_00_r(ComponentPose=@n19, Alpha=1, ExternalForce="0,0,0") {
    angularDampingOverride=1
    angularSpringConstant=15000
    bAngularSpring=true
    bOverrideAngularDamping=true
    bUseGravityOverride=true
    gravityScale=0.35
    simulationSpace=BoneRelative
    alphaBoolBlend {blendOption=Linear}
    boundBone {boneName=dyn_frontBracelete_00_r}
    componentAppliedLinearAccClamp=<3000 3000 3000>
    componentLinearAccScale=<0.5 0.5 0.5>
    componentLinearVelScale=<0.5 0.5 0.5>
    gravityOverride {z=-980}
    physicsBodyDefinitions [{
      boundBone {boneName=dyn_frontBracelete_00_r}
      boxExtents=<25 25 25>
      constraintSetup {angularLimitsMax=<0 0 3>, angularLimitsMin=<0 0 -2>, angularTarget=<1 0 0>}
      localJointOffset=<4 0 0>
     }]
    relativeSpaceBone {boneName=root}
   }
   n21 = Anim Dynamics - Bone: dyn_backBracelete_00_r(ComponentPose=@n20, Alpha=1, ExternalForce="0,0,0") {
    angularDampingOverride=1
    angularSpringConstant=10000
    bAngularSpring=true
    bOverrideAngularDamping=true
    bUseGravityOverride=true
    gravityScale=0.35
    simulationSpace=BoneRelative
    alphaBoolBlend {blendOption=Linear}
    boundBone {boneName=dyn_backBracelete_00_r}
    componentAppliedLinearAccClamp=<3000 3000 3000>
    componentLinearAccScale=<0.5 0.5 0.5>
    componentLinearVelScale=<0.5 0.5 0.5>
    gravityOverride {z=-980}
    physicsBodyDefinitions [{
      sphereCollisionRadius=0
      boundBone {boneName=dyn_backBracelete_00_r}
      boxExtents=<30 30 30>
      constraintSetup {angularLimitsMax=<0 0 5>, angularLimitsMin=<0 0 -6>, angularTarget=<1 0 0>}
      localJointOffset=<4 0 0>
     }]
    relativeSpaceBone {boneName=root}
   }
   n22 = Anim Dynamics - Bone: dyn_sleeve_00_l(ComponentPose=@n21, Alpha=1, ExternalForce="0,0,0") {
    angularDampingOverride=1
    angularSpringConstant=1600
    bAngularSpring=true
    bUseGravityOverride=true
    gravityScale=0.35
    simulationSpace=BoneRelative
    alphaBoolBlend {blendOption=Linear}
    boundBone {boneName=dyn_sleeve_00_l}
    componentLinearAccScale=<1 1 1>
    componentLinearVelScale=<1 1 1>
    gravityOverride {z=-980}
    physicsBodyDefinitions [{
      boundBone {boneName=dyn_sleeve_00_l}
      boxExtents=<40 40 40>
      constraintSetup {angularLimitsMax=<0 0 20>, angularLimitsMin=<0 0 0>, angularTarget=<1 0 0>}
      localJointOffset=<4 0 0>
     }]
    relativeSpaceBone {boneName=root}
   }
   n23 = Anim Dynamics - Bone: dyn_sleeve_01_l(ComponentPose=@n22, Alpha=1, ExternalForce="0,0,0") {
    angularDampingOverride=1
    angularSpringConstant=2000
    bAngularSpring=true
    bUseGravityOverride=true
    gravityScale=0.35
    simulationSpace=BoneRelative
    alphaBoolBlend {blendOption=Linear}
    boundBone {boneName=dyn_sleeve_01_l}
    componentLinearAccScale=<1 1 1>
    componentLinearVelScale=<1 1 1>
    gravityOverride {z=-980}
    physicsBodyDefinitions [{
      boundBone {boneName=dyn_sleeve_01_l}
      boxExtents=<24 24 24>
      constraintSetup {angularLimitsMax=<0 0 0>, angularLimitsMin=<0 0 -45>, angularTarget=<1 0 0>}
      localJointOffset=<4 0 0>
     }]
    relativeSpaceBone {boneName=root}
   }
   n24 = Anim Dynamics - Bone: dyn_sleeve_00_r(ComponentPose=@n23, Alpha=1, ExternalForce="0,0,0") {
    angularDampingOverride=1
    angularSpringConstant=1600
    bAngularSpring=true
    bUseGravityOverride=true
    gravityScale=0.35
    simulationSpace=BoneRelative
    alphaBoolBlend {blendOption=Linear}
    boundBone {boneName=dyn_sleeve_00_r}
    componentLinearAccScale=<1 1 1>
    componentLinearVelScale=<1 1 1>
    gravityOverride {z=-980}
    physicsBodyDefinitions [{
      boundBone {boneName=dyn_sleeve_00_r}
      boxExtents=<45 45 45>
      constraintSetup {angularLimitsMax=<0 0 0>, angularLimitsMin=<0 0 -20>, angularTarget=<1 0 0>}
      localJointOffset=<4 0 0>
     }]
    relativeSpaceBone {boneName=root}
   }
   n25 = Anim Dynamics - Bone: dyn_sleeve_01_r(ComponentPose=@n24, Alpha=1, ExternalForce="0,0,0") {
    angularDampingOverride=1
    angularSpringConstant=2000
    bAngularSpring=true
    bUseGravityOverride=true
    gravityScale=0.35
    simulationSpace=BoneRelative
    alphaBoolBlend {blendOption=Linear}
    boundBone {boneName=dyn_sleeve_01_r}
    componentLinearAccScale=<1 1 1>
    componentLinearVelScale=<1 1 1>
    gravityOverride {z=-980}
    physicsBodyDefinitions [{
      boundBone {boneName=dyn_sleeve_01_r}
      boxExtents=<22.5 22.5 22.5>
      constraintSetup {angularLimitsMax=<0 0 40>, angularLimitsMin=<0 0 0>, angularTarget=<1 0 0>}
      localJointOffset=<4 0 0>
     }]
    relativeSpaceBone {boneName=root}
   }
   n26 = Anim Dynamics - Bone: dyn_knife(ComponentPose=@n25, Alpha=1, ExternalForce="0,0,0") {
    angularDampingOverride=1
    angularSpringConstant=20000
    bAngularSpring=true
    bOverrideAngularDamping=true
    bUseGravityOverride=true
    gravityScale=0.35
    simulationSpace=BoneRelative
    alphaBoolBlend {blendOption=Linear}
    boundBone {boneName=dyn_knife}
    componentAppliedLinearAccClamp=<3000 3000 3000>
    componentLinearAccScale=<0.3 0.3 0.3>
    componentLinearVelScale=<0.3 0.3 0.3>
    gravityOverride {z=-980}
    physicsBodyDefinitions [{
      boundBone {boneName=dyn_knife}
      boxExtents=<30 30 30>
      constraintSetup {angularLimitsMax=<0 0 5>, angularLimitsMin=<0 0 -5>, angularTarget=<1 0 0>}
      localJointOffset=<4 0 0>
     }]
    relativeSpaceBone {boneName=root}
   }
   n27 = Anim Dynamics - Bone: dyn_cannisters(ComponentPose=@n26, Alpha=1, ExternalForce="0,0,0") {
    angularDampingOverride=1
    angularSpringConstant=20000
    bAngularSpring=true
    bOverrideAngularDamping=true
    bUseGravityOverride=true
    gravityScale=0.35
    simulationSpace=RootRelative
    alphaBoolBlend {blendOption=Linear}
    boundBone {boneName=dyn_cannisters}
    componentAppliedLinearAccClamp=<3000 3000 3000>
    componentLinearAccScale=<1 1 1>
    componentLinearVelScale=<1 1 1>
    gravityOverride {z=-980}
    physicsBodyDefinitions [{
      boundBone {boneName=dyn_cannisters}
      boxExtents=<10 10 10>
      constraintSetup {angularLimitsMax=<0 0 5>, angularLimitsMin=<0 0 -5>, angularTarget=<0 0 0>}
      localJointOffset=<4 0 0>
     }]
   }
   n28 = Anim Dynamics - Bone: dyn_pantspull_l(ComponentPose=@n27, Alpha=1, ExternalForce="0,0,0") {
    angularDampingOverride=1
    angularSpringConstant=20000
    bAngularSpring=true
    bOverrideAngularDamping=true
    bUseGravityOverride=true
    gravityScale=0.35
    simulationSpace=RootRelative
    alphaBoolBlend {blendOption=Linear}
    boundBone {boneName=dyn_pantspull_l}
    componentLinearAccScale=<1 1 1>
    componentLinearVelScale=<1 1 1>
    gravityOverride {z=-980}
    physicsBodyDefinitions [{
      boundBone {boneName=dyn_pantspull_l}
      boxExtents=<15 15 15>
      constraintSetup {angularLimitsMax=<0 30 0>, angularLimitsMin=<0 -30 0>, angularTarget=<0 0 0>}
      localJointOffset=<4 0 0>
     }]
   }
   Save cached pose 'animDynCache'(Pose=Component To Local(ComponentPose=@n28)) {cacheName=animDynCache}
   n29 = Control Rig(Source=Use cached pose 'avoid_clipping_sleeves_cache'()) {
    alphaBoolBlend {blendOption=Linear}
    controlRigAssetReference {blueprintRigClass=&/M_MED_PitGlass_DeformRig_Master.M_MED_PitGlass_DeformRig_Master_C}
   }
   n30 = "Transform (Modify) Bone - Bone: base_zipper_00_r"(
    ComponentPose=Local To Component(LocalPose=@n29)
    Alpha=Get avoid_clipping_target_up() -> (avoid_clipping_target_up:double)
    Translation="0,0,0"
    Rotation="80,0,0"
    Scale="1,1,1"
   ) {
    rotationMode=BMM_Additive
    rotationSpace=BCS_BoneSpace
    alphaBoolBlend {blendOption=Linear}
    alphaScaleBiasClamp {bInterpResult=true}
    boneToModify {boneName=base_zipper_00_r}
    rotation {pitch=80}
   }
   n31 = Anim Dynamics - Bone: dyn_zipper_00_l(ComponentPose=@n30, Alpha=1, ExternalForce="0,0,0") {
    angularDampingOverride=1
    angularSpringConstant=5000
    bAngularSpring=true
    bOverrideAngularDamping=true
    bUseGravityOverride=true
    simulationSpace=BoneRelative
    alphaBoolBlend {blendOption=Linear}
    boundBone {boneName=dyn_zipper_00_l}
    componentAppliedLinearAccClamp=<3000 3000 3000>
    componentLinearAccScale=<0.5 0.5 0.5>
    componentLinearVelScale=<0.5 0.5 0.5>
    gravityOverride {z=-980}
    physicsBodyDefinitions [{
      boundBone {boneName=dyn_zipper_00_l}
      boxExtents=<1 25 10>
      constraintSetup {angularLimitsMax=<0 0 20>, angularLimitsMin=<0 -20 -20>, angularTarget=<1 0 0>}
      localJointOffset=<2 0 3>
     }]
    relativeSpaceBone {boneName=root}
   }
   n32 = Anim Dynamics - Bone: dyn_zipper_01_l(ComponentPose=@n31, Alpha=1, ExternalForce="0,0,0") {
    angularDampingOverride=1
    angularSpringConstant=3000
    bAngularSpring=true
    bOverrideAngularDamping=true
    bUseGravityOverride=true
    gravityScale=0.35
    simulationSpace=BoneRelative
    alphaBoolBlend {blendOption=Linear}
    boundBone {boneName=dyn_zipper_01_l}
    componentAppliedLinearAccClamp=<3000 3000 3000>
    componentLinearAccScale=<0.5 0.5 0.5>
    componentLinearVelScale=<0.5 0.5 0.5>
    gravityOverride {z=-980}
    physicsBodyDefinitions [{
      boundBone {boneName=dyn_zipper_01_l}
      boxExtents=<1 25 10>
      constraintSetup {angularLimitsMax=<0 0 20>, angularLimitsMin=<0 -30 -20>, angularTarget=<1 0 0>}
      localJointOffset=<2 0 3>
     }]
    relativeSpaceBone {boneName=root}
   }
   n33 = Anim Dynamics - Bone: dyn_zipper_01_r(ComponentPose=@n32, Alpha=1, ExternalForce="0,0,0") {
    angularDampingOverride=1
    angularSpringConstant=3500
    bAngularSpring=true
    bOverrideAngularDamping=true
    bUseGravityOverride=true
    gravityScale=0.35
    simulationSpace=BoneRelative
    alphaBoolBlend {blendOption=Linear}
    boundBone {boneName=dyn_zipper_01_r}
    componentAppliedLinearAccClamp=<3000 3000 3000>
    componentLinearAccScale=<0.5 0.5 0.5>
    componentLinearVelScale=<0.5 0.5 0.5>
    gravityOverride {z=-980}
    physicsBodyDefinitions [{
      boundBone {boneName=dyn_zipper_01_r}
      boxExtents=<1 25 10>
      constraintSetup {angularLimitsMax=<0 0 15>, angularLimitsMin=<0 -30 -15>, angularTarget=<1 0 0>}
      localJointOffset=<2 0 3>
     }]
    relativeSpaceBone {boneName=root}
   }
   n34 = Anim Dynamics - Bone: dyn_zipper_00_r(ComponentPose=@n33, Alpha=1, ExternalForce="0,0,0") {
    angularDampingOverride=1
    angularSpringConstant=4000
    bAngularSpring=true
    bOverrideAngularDamping=true
    bUseGravityOverride=true
    gravityScale=0.35
    simulationSpace=BoneRelative
    alphaBoolBlend {blendOption=Linear}
    boundBone {boneName=dyn_zipper_00_r}
    componentAppliedLinearAccClamp=<3000 3000 3000>
    componentLinearAccScale=<0.5 0.5 0.5>
    componentLinearVelScale=<0.5 0.5 0.5>
    gravityOverride {z=-980}
    physicsBodyDefinitions [{
      boundBone {boneName=dyn_zipper_00_r}
      boxExtents=<1 25 10>
      constraintSetup {angularLimitsMax=<0 0 15>, angularLimitsMin=<0 -30 -15>, angularTarget=<1 0 0>}
      localJointOffset=<2 0 3>
     }]
    relativeSpaceBone {boneName=root}
   }
   n35 = "Transform (Modify) Bone - Bone: base_gun_0_l"(
    ComponentPose=@n34
    Alpha=Get avoid_clippingGun_l_skydiving() -> (avoid_clippingGun_l_skydiving:double)
    Translation="8,0,5"
    Rotation="25,0,0"
    Scale="1,1,1"
   ) {
    rotationMode=BMM_Additive
    rotationSpace=BCS_BoneSpace
    translationMode=BMM_Additive
    translationSpace=BCS_BoneSpace
    alphaBoolBlend {blendOption=Linear}
    boneToModify {boneName=base_gun_0_l}
    rotation {pitch=25}
    translation {x=8, z=5}
   }
   n36 = "Transform (Modify) Bone - Bone: base_gun_0_r"(
    ComponentPose=@n35
    Alpha=Get avoid_clippingGun_r_skydiving() -> (avoid_clippingGun_r_skydiving:double)
    Translation="8,0,5"
    Rotation="25,0,0"
    Scale="1,1,1"
   ) {
    rotationMode=BMM_Additive
    rotationSpace=BCS_BoneSpace
    translationMode=BMM_Additive
    translationSpace=BCS_BoneSpace
    alphaBoolBlend {blendOption=Linear}
    boneToModify {boneName=base_gun_0_r}
    rotation {pitch=25}
    translation {x=8, z=5}
   }
   n37 = Get avoid_clipping_target_up()
   n38 = "Transform (Modify) Bone - Bone: base_gun_0_r"(ComponentPose=@n36, Alpha=@n37, Translation="0,0,0.5", Rotation="10.5,0,12", Scale="1,1,1") {
    rotationMode=BMM_Additive
    rotationSpace=BCS_BoneSpace
    translationMode=BMM_Additive
    translationSpace=BCS_BoneSpace
    alphaBoolBlend {blendOption=Linear}
    alphaScaleBiasClamp {bInterpResult=true, interpSpeedDecreasing=1, interpSpeedIncreasing=11}
    boneToModify {boneName=base_gun_0_r}
    rotation {pitch=10.5, roll=12}
    translation {z=0.5}
   }
   Save cached pose 'animDyn_Guns_Cache'(Pose=Component To Local(ComponentPose=@n38)) {cacheName=animDyn_Guns_Cache}
   n39 = "Transform (Modify) Bone - Bone: base_sleeve_00_r"(
    ComponentPose=Local To Component(LocalPose=Use cached pose 'sin_animation_sharps_r_cache'())
    Alpha=Get avoid_clipping_sleeve_r() -> (avoid_clipping_sleeve_r:double)
    Translation="0,0,0"
    Rotation=Get avoid_clipping_sleeve_isFloating() -> (avoid_clipping_sleeve_isFloating:Rotator)
    Scale="1,1,1"
   ) {
    rotationMode=BMM_Additive
    rotationSpace=BCS_BoneSpace
    translationMode=BMM_Additive
    translationSpace=BCS_BoneSpace
    alphaBoolBlend {blendOption=Linear}
    boneToModify {boneName=base_sleeve_00_r}
    rotation {pitch=-20, yaw=-60}
   }
   n40 = Get avoid_clipping_sleeve_parachute() -> (avoid_clipping_sleeve_parachute:double)
   n41 = "Transform (Modify) Bone - Bone: base_sleeve_01_r"(ComponentPose=@n39, Alpha=@n40, Translation="-1,0,0", Rotation="0,0,0", Scale="1,1,1") {
    rotationMode=BMM_Additive
    rotationSpace=BCS_BoneSpace
    translationMode=BMM_Additive
    translationSpace=BCS_BoneSpace
    alphaBoolBlend {blendOption=Linear}
    alphaScaleBiasClamp {bInterpResult=true}
    boneToModify {boneName=base_sleeve_01_r}
    translation {x=-1}
   }
   n42 = "Transform (Modify) Bone - Bone: base_sleeve_01_l"(ComponentPose=@n41, Alpha=@n40, Translation="-1,0,0", Rotation="0,0,0", Scale="1,1,1") {
    rotationMode=BMM_Additive
    rotationSpace=BCS_BoneSpace
    translationMode=BMM_Additive
    translationSpace=BCS_BoneSpace
    alphaBoolBlend {blendOption=Linear}
    alphaScaleBiasClamp {bInterpResult=true}
    boneToModify {boneName=base_sleeve_01_l}
    translation {x=-1}
   }
   n43 = "Transform (Modify) Bone - Bone: base_sleeve_00_l"(
    ComponentPose=@n42
    Alpha=Get avoid_clipping_sleeve_l() -> (avoid_clipping_sleeve_l:double)
    Translation="0,0,0"
    Rotation="20,60,-20"
    Scale="1,1,1"
   ) {
    rotationMode=BMM_Additive
    rotationSpace=BCS_BoneSpace
    translationMode=BMM_Additive
    translationSpace=BCS_BoneSpace
    alphaBoolBlend {blendOption=Linear}
    boneToModify {boneName=base_sleeve_00_l}
    rotation=<rot 20 60 -20>
   }
   n44 = "Transform (Modify) Bone - Bone: base_sleeve_00_r"(ComponentPose=@n43, Alpha=@n37, Translation="0,-1,0", Rotation="0,-10,0", Scale="1,1,1") {
    rotationMode=BMM_Additive
    rotationSpace=BCS_BoneSpace
    translationMode=BMM_Additive
    translationSpace=BCS_BoneSpace
    alphaBoolBlend {blendOption=Linear}
    alphaScaleBiasClamp {bInterpResult=true}
    boneToModify {boneName=base_sleeve_00_r}
    rotation {yaw=-10}
    translation {y=-1}
   }
   n45 = "Transform (Modify) Bone - Bone: base_sleeve_01_r"(ComponentPose=@n44, Alpha=@n37, Translation="-2,0,0", Rotation="0,10,0", Scale="1,1,1") {
    rotationMode=BMM_Additive
    rotationSpace=BCS_BoneSpace
    translationMode=BMM_Additive
    translationSpace=BCS_BoneSpace
    alphaBoolBlend {blendOption=Linear}
    alphaScaleBiasClamp {bInterpResult=true}
    boneToModify {boneName=base_sleeve_01_r}
    rotation {yaw=10}
    translation {x=-2}
   }
   Save cached pose 'avoid_clipping_sleeves_cache'(Pose=Component To Local(ComponentPose=@n45)) {cacheName=avoid_clipping_sleeves_cache}
   n46 = "Transform (Modify) Bone - Bone: dyn_sharp_00_l"(
    ComponentPose=Local To Component(LocalPose=Use cached pose 'SurfaceConstraintsCache'())
    Alpha=1
    Translation=Get Translate Shard_00_l() -> (Translate Shard_00_l:Vector)
    Rotation="0,0,0"
    Scale="1,1,1"
   ) {translationMode=BMM_Additive, translationSpace=BCS_BoneSpace, alphaBoolBlend {blendOption=Linear}, boneToModify {boneName=dyn_sharp_00_l}}
   n47 = "Transform (Modify) Bone - Bone: dyn_sharp_01_l"(ComponentPose=@n46, Alpha=1, Translation=Get Translate Shard_01_l() -> (Translate Shard_01_l:Vector), Rotation="0,0,0", Scale="1,1,1") {translationMode=BMM_Additive, translationSpace=BCS_BoneSpace, alphaBoolBlend {blendOption=Linear}, boneToModify {boneName=dyn_sharp_01_l}}
   n48 = "Transform (Modify) Bone - Bone: dyn_sharp_02_l"(ComponentPose=@n47, Alpha=1, Translation=Get Translate Shard_02_l() -> (Translate Shard_02_l:Vector), Rotation="0,0,0", Scale="1,1,1") {translationMode=BMM_Additive, translationSpace=BCS_BoneSpace, alphaBoolBlend {blendOption=Linear}, boneToModify {boneName=dyn_sharp_02_l}}
   n49 = "Transform (Modify) Bone - Bone: dyn_sharp_03_l"(ComponentPose=@n48, Alpha=1, Translation=Get Translate Shard_03_l() -> (Translate Shard_03_l:Vector), Rotation="0,0,0", Scale="1,1,1") {translationMode=BMM_Additive, translationSpace=BCS_BoneSpace, alphaBoolBlend {blendOption=Linear}, boneToModify {boneName=dyn_sharp_03_l}}
   Save cached pose 'sin_animation_sharps_l_cache'(Pose=Component To Local(ComponentPose=@n49)) {cacheName=sin_animation_sharps_l_cache}
   n50 = "Transform (Modify) Bone - Bone: dyn_sharp_00_r"(
    ComponentPose=Local To Component(LocalPose=Use cached pose 'sin_animation_sharps_l_cache'())
    Alpha=1
    Translation=Get Translate Shard_00_r() -> (Translate Shard_00_r:Vector)
    Rotation="0,0,0"
    Scale="1,1,1"
   ) {translationMode=BMM_Additive, translationSpace=BCS_BoneSpace, alphaBoolBlend {blendOption=Linear}, boneToModify {boneName=dyn_sharp_00_r}}
   n51 = "Transform (Modify) Bone - Bone: dyn_sharp_01_r"(ComponentPose=@n50, Alpha=1, Translation=Get Translate Shard_01_r() -> (Translate Shard_01_r:Vector), Rotation="0,0,0", Scale="1,1,1") {translationMode=BMM_Additive, translationSpace=BCS_BoneSpace, alphaBoolBlend {blendOption=Linear}, boneToModify {boneName=dyn_sharp_01_r}}
   n52 = "Transform (Modify) Bone - Bone: dyn_sharp_02_r"(ComponentPose=@n51, Alpha=1, Translation=Get Translate Shard_02_r() -> (Translate Shard_02_r:Vector), Rotation="0,0,0", Scale="1,1,1") {translationMode=BMM_Additive, translationSpace=BCS_BoneSpace, alphaBoolBlend {blendOption=Linear}, boneToModify {boneName=dyn_sharp_02_r}}
   n53 = "Transform (Modify) Bone - Bone: dyn_sharp_03_r"(ComponentPose=@n52, Alpha=1, Translation=Get Translate Shard_03_r() -> (Translate Shard_03_r:Vector), Rotation="0,0,0", Scale="1,1,1") {translationMode=BMM_Additive, translationSpace=BCS_BoneSpace, alphaBoolBlend {blendOption=Linear}, boneToModify {boneName=dyn_sharp_03_r}}
   Save cached pose 'sin_animation_sharps_r_cache'(Pose=Component To Local(ComponentPose=@n53)) {cacheName=sin_animation_sharps_r_cache}
  ]}
 EventGraph {nodes [
   Event BlueprintUpdateAnimation()
   > n1 = Sequence!(execute) -> (then_0, then_1, then_2:Exec, then_3:Exec, then_4:Exec, then_5:Exec, then_6:Exec, then_7:Exec, then_8:Exec, then_9:Exec, then_10:Exec)
   Set avoid_clippingGun_l_skydiving(execute=@n1.then_1, avoid_clippingGun_l_skydiving:double="To Float (Boolean)"(InBool=Get bIsSkyDiving())) -> (Output_Get:double)
   n2 = Set Skel Mesh(execute=@n1.then_5, Skel Mesh=GetOwnerSkeletalMeshComponent())
   n3 = Set avoid_clipping_target_up(
    execute=@n1.then_10
    avoid_clipping_target_up:double=GetAlphaFromReferenceTransformDelta(SkeletalMeshComponent=@n2, BoneName=neck_02, BaseName=spine_05, InputRange=(X=0,Y=-20.5), TwistAxis="0,0,1")
   ) -> (Output_Get:double)
   > n4 = Branch(Condition=Get bIsTargeting())
   > Set avoid_clipping_target_up(avoid_clipping_target_up=@n3)
   Set avoid_clipping_target_up(execute=@n4.else, avoid_clipping_target_up=0)
   Set avoid_clippingBracelete_targeting(execute=@n1.then_2, avoid_clippingBracelete_targeting:double="To Float (Boolean)"(InBool=Get bIsTargeting())) -> (Output_Get:double)
   n5 = BreakTransform(InTransform=GetDeltaTransformFromRefPose(self=GetOwnerSkeletalMeshComponent(), BoneName=thigh_r, BaseName=pelvis))
   > Set thigh_rot_up_r(thigh_rot_up_r:double=MapRangeClamped(Value=BreakRotator(InRot=@n5.Rotation).Yaw, InRangeA=0, InRangeB=-100)) -> (Output_Get:double)
   n6 = "Sin (Radians)"(A=float * float(A=GetRealTimeSeconds(), B=4.3))
   Set Translate Shard_00_l(execute=@n1.then_3, Translate Shard_00_l:Vector=MakeVector(X=float / float(A=@n6, B=-4) {is=title K2Node_CallFunction}, Y=0)) -> (Output_Get:Vector)
   n7 = "Sin (Radians)"(A=float * float(A=GetRealTimeSeconds(), B=3.2))
   > Set Translate Shard_01_l(Translate Shard_01_l:Vector=MakeVector(X=float / float(A=@n7, B=-2.8), Y=0)) -> (Output_Get:Vector)
   n8 = "Sin (Radians)"(A=float * float(A=GetRealTimeSeconds(), B=4))
   > Set Translate Shard_02_l(Translate Shard_02_l:Vector=MakeVector(X=float / float(A=@n8, B=-2.5), Y=0)) -> (Output_Get:Vector)
   n9 = "Sin (Radians)"(A=float * float(A=GetRealTimeSeconds(), B=5))
   > Set Translate Shard_03_l(Translate Shard_03_l:Vector=MakeVector(X=float / float(A=@n9, B=-8), Y=0)) -> (Output_Get:Vector)
   n10 = Get bIsParachuteOpened()
   n11 = Get bIsInFrontEnd()
   Set avoid_clipping_sleeve_parachute(execute=@n1.then_7, avoid_clipping_sleeve_parachute:double="To Float (Boolean)"(InBool=OR Boolean(A=@n10, B=@n11))) -> (Output_Get:double)
   n12 = Branch(execute=@n1.then_8, Condition=Get bIsSkydiveFloating())
   > Set avoid_clipping_sleeve_isFloating(avoid_clipping_sleeve_isFloating:Rotator="0,-40,0") -> (Output_Get:Rotator)
   Set avoid_clipping_sleeve_isFloating(execute=@n12.else, avoid_clipping_sleeve_isFloating="-20,-90,0")
   Set Knee_Brace(
    execute=@n1.then_9
    Knee_Brace:double=GetAlphaFromReferenceTransformDelta(SkeletalMeshComponent=@n2, BoneName=calf_r, BaseName=thigh_r, InputRange=(X=0,Y=120), TwistAxis="0,0,1")
   ) -> (Output_Get:double)
   n13 = Gravity Override State Function DEPRECATED(
    execute=@n1.then_0
    CharacterPartAnimInstance=Self-Reference()
    GameTimeInSeconds=GetGameTimeInSeconds()
    GravityOverrideParams=Get Gravity Override Params()
   )
   > Set GravityOverride(GravityOverride=@n13)
   n14 = "Sin (Radians)"(A=float * float(A=GetRealTimeSeconds(), B=4.3))
   Set Translate Shard_00_r(execute=@n1.then_4, Translate Shard_00_r:Vector=MakeVector(X=float / float(A=@n14, B=-4.5), Y=0)) -> (Output_Get:Vector)
   n15 = "Sin (Radians)"(A=float * float(A=GetRealTimeSeconds(), B=2.8))
   > Set Translate Shard_01_r(Translate Shard_01_r:Vector=MakeVector(X=float / float(A=@n15, B=-2.8), Y=0)) -> (Output_Get:Vector)
   n16 = "Sin (Radians)"(A=float * float(A=GetRealTimeSeconds(), B=3.6))
   > Set Translate Shard_02_r(Translate Shard_02_r:Vector=MakeVector(X=float / float(A=@n16, B=-3.2), Y=0)) -> (Output_Get:Vector)
   n17 = "Sin (Radians)"(A=float * float(A=GetRealTimeSeconds(), B=2.1))
   > Set Translate Shard_03_r(Translate Shard_03_r:Vector=MakeVector(X=float / float(A=@n17, B=-2.5), Y=0)) -> (Output_Get:Vector)
   Set avoid_clippingGun_l_skydiving(
    execute=@n1.then_6
    avoid_clippingGun_l_skydiving=GetAlphaFromReferenceTransformDelta(SkeletalMeshComponent=@n2, BoneName=clavicle_l, BaseName=upperarm_l, DeltaComponentType=EulerY, InputRange=(X=0,Y=-170), TwistAxis="0,0,1")
   )
   > Set avoid_clippingGun_r_skydiving(avoid_clippingGun_r_skydiving:double=GetAlphaFromReferenceTransformDelta(SkeletalMeshComponent=@n2, BoneName=clavicle_r, BaseName=upperarm_r, DeltaComponentType=EulerY, InputRange=(X=0,Y=-170), TwistAxis="0,0,1")) -> (Output_Get:double)
   Set avoid_clipping_sleeve_l(
    execute=@n19.then
    avoid_clipping_sleeve_l:double=GetAlphaFromReferenceTransformDelta(SkeletalMeshComponent=@n2, BoneName=upperarm_l, BaseName=lowerarm_l, InputRange=(X=0,Y=-120), TwistAxis="0,0,1")
   ) -> (Output_Get:double)
   n18 = GetAlphaFromReferenceTransformDelta(SkeletalMeshComponent=@n2, BoneName=upperarm_r, BaseName=lowerarm_r, InputRange=(X=0,Y=-120), TwistAxis="0,0,1")
   n19 = Set avoid_clipping_sleeve_r(execute=@n2.then, avoid_clipping_sleeve_r:double=@n18) -> (Output_Get:double)
  ]}
]
variables [
 avoid_clipping_sleeve_isFloating {type=Rotator}
 avoid_clipping_sleeve_l {type=number}
 avoid_clipping_sleeve_parachute {type=number}
 avoid_clipping_sleeve_r {type=number}
 avoid_clipping_target_up {type=number}
 avoid_clippingBracelete_targeting {type=number}
 avoid_clippingGun_l_skydiving {type=number}
 avoid_clippingGun_r_skydiving {type=number}
 Gravity Override Params {type=GravityOverrideParamsStruct}
 GravityOverride {type=Vector}
 Knee_Brace {type=number}
 Skel Mesh {type=SkeletalMeshComponent}
 thigh_rot_up_r {type=number}
 Translate Shard_00_l {type=Vector}
 Translate Shard_00_r {type=Vector}
 Translate Shard_01_l {type=Vector}
 Translate Shard_01_r {type=Vector}
 Translate Shard_02_l {type=Vector}
 Translate Shard_02_r {type=Vector}
 Translate Shard_03_l {type=Vector}
 Translate Shard_03_r {type=Vector}
]
