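# Physics asset for the PitGlass medium male character's face accessory part, covering jiggle bones for hair, an earring and two hoodie pieces.
# Each dynamic capsule body is constrained to a small base anchor capsule or to the preceding hair link.
# Hair strands 00–06 form two-link base-to-ini-to-tip chains, while 07–08 have one link.
# Swing limits are mostly tight with near-zero twist, with wider swing on the earring and hoodie.
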PhysicsAsset /BRCosmetics/Characters/Player/Male/Medium/Bodies/M_MED_PitGlass/Meshes/Parts/M_MED_PitGlass_FaceAcc_Physics
bodies {
 base_B_hoodie [{shapeType=Capsule, length=0.505, radius=0.505, shapeName=base_B_hoodie_capsule, rotation=<rot 0 -90 0>}]
 base_earring [{shapeType=Capsule, length=0.505, radius=0.505, shapeName=base_earring_capsule, rotation=<rot 0 -90 0>}]
 base_hair_ini_00 [{shapeType=Capsule, length=0.505, radius=0.505, shapeName=base_hair_ini_00_capsule, rotation=<rot 0 -90 0>}]
 base_hair_ini_01 [{shapeType=Capsule, length=0.505, radius=0.505, shapeName=base_hair_ini_01_capsule, rotation=<rot 0 -90 0>}]
 base_hair_ini_02 [{shapeType=Capsule, length=0.505, radius=0.505, shapeName=base_hair_ini_02_capsule, rotation=<rot 0 -90 0>}]
 base_hair_ini_03 [{shapeType=Capsule, length=0.505, radius=0.505, shapeName=base_hair_ini_03_capsule, rotation=<rot 0 -90 0>}]
 base_hair_ini_04 [{shapeType=Capsule, length=0.505, radius=0.505, shapeName=base_hair_ini_04_capsule, rotation=<rot 0 -90 0>}]
 base_hair_ini_05 [{shapeType=Capsule, length=0.505, radius=0.505, shapeName=base_hair_ini_05_capsule, rotation=<rot 0 -90 0>}]
 base_hair_ini_06 [{shapeType=Capsule, length=0.505, radius=0.505, shapeName=base_hair_ini_06_capsule, rotation=<rot 0 -90 0>}]
 base_hair_ini_07 [{shapeType=Capsule, length=0.505, radius=0.505, shapeName=base_hair_ini_07_capsule, rotation=<rot 0 -90 0>}]
 base_hair_ini_08 [{shapeType=Capsule, length=0.505, radius=0.505, shapeName=base_hair_ini_08_capsule, rotation=<rot 0 -90 0>}]
 base_hoodie [{shapeType=Capsule, length=0.505, radius=0.505, shapeName=base_hoodie_capsule, rotation=<rot 0 -90 0>}]
 dyn_B_hoodie [{shapeType=Capsule, length=2, radius=2, shapeName=dyn_B_hoodie_capsule, center=<1.3 0 -1.5>, rotation=<rot 90 0 0>}]
 dyn_earring [{shapeType=Capsule, length=2, radius=1, shapeName=dyn_earring_capsule, center=<2 0 0>, rotation=<rot 90 0 0>}]
 dyn_hair_ini_00 [{shapeType=Capsule, length=2, radius=1, shapeName=dyn_hair_ini_00_capsule, center=<3 0 0>, rotation=<rot 90 0 0>}]
 dyn_hair_ini_01 [{shapeType=Capsule, length=2, radius=1, shapeName=dyn_hair_ini_01_capsule, center=<2 0 0>, rotation=<rot 90 0 0>}]
 dyn_hair_ini_02 [{shapeType=Capsule, length=2, radius=1, shapeName=dyn_hair_ini_02_capsule, center=<0.5 0 0>, rotation=<rot 90 0 0>}]
 dyn_hair_ini_03 [{shapeType=Capsule, length=2, radius=1, shapeName=dyn_hair_ini_03_capsule, center=<2 0 0>, rotation=<rot 90 0 0>}]
 dyn_hair_ini_04 [{shapeType=Capsule, length=2, radius=1, shapeName=dyn_hair_ini_04_capsule, center=<2 0 0>, rotation=<rot 90 0 0>}]
 dyn_hair_ini_05 [{shapeType=Capsule, length=2, radius=1, shapeName=dyn_hair_ini_05_capsule, center=<2 0 0>, rotation=<rot 90 0 0>}]
 dyn_hair_ini_06 [{shapeType=Capsule, length=2, radius=1, shapeName=dyn_hair_ini_06_capsule, center=<2 0 0>, rotation=<rot 90 0 0>}]
 dyn_hair_ini_07 [{shapeType=Capsule, length=1, radius=1, shapeName=dyn_hair_ini_07_capsule, center=<2 0 2>, rotation=<rot 90 0 0>}]
 dyn_hair_ini_08 [{shapeType=Capsule, length=1, radius=1, shapeName=dyn_hair_ini_08_capsule, center=<2 0 2>, rotation=<rot 90 0 0>}]
 dyn_hair_tip_00 [{shapeType=Capsule, length=2, radius=1, shapeName=dyn_hair_tip_00_capsule, center=<3 0 0>, rotation=<rot 90 0 0>}]
 dyn_hair_tip_01 [{shapeType=Capsule, length=2, radius=1, shapeName=dyn_hair_tip_01_capsule, center=<2 0 0>, rotation=<rot 90 0 0>}]
 dyn_hair_tip_02 [{shapeType=Capsule, length=2, radius=1, shapeName=dyn_hair_tip_02_capsule, center=<0.5 0 0>, rotation=<rot 90 0 0>}]
 dyn_hair_tip_03 [{shapeType=Capsule, length=2, radius=1, shapeName=dyn_hair_tip_03_capsule, center=<2 0 0>, rotation=<rot 90 0 0>}]
 dyn_hair_tip_04 [{shapeType=Capsule, length=2, radius=1, shapeName=dyn_hair_tip_04_capsule, center=<2 0 0>, rotation=<rot 90 0 0>}]
 dyn_hair_tip_05 [{shapeType=Capsule, length=2, radius=1, shapeName=dyn_hair_tip_05_capsule, center=<2 0 0>, rotation=<rot 90 0 0>}]
 dyn_hair_tip_06 [{shapeType=Capsule, length=2, radius=1, shapeName=dyn_hair_tip_06_capsule, center=<2 0 0>, rotation=<rot 90 0 0>}]
 dyn_hoodie [{shapeType=Capsule, length=2, radius=1, shapeName=dyn_hoodie_capsule, center=<2 0 0>, rotation=<rot 90 0 0>}]
}
constraints [
 {bone1Name=dyn_earring, bone2Name=base_earring, swing1LimitDegrees=0, swing2LimitDegrees=15, twistLimitDegrees=1}
 {bone1Name=dyn_hair_ini_00, bone2Name=base_hair_ini_00, swing1LimitDegrees=3, swing2LimitDegrees=3, twistLimitDegrees=1}
 {bone1Name=dyn_hair_tip_00, bone2Name=dyn_hair_ini_00, swing1LimitDegrees=3, swing2LimitDegrees=3, twistLimitDegrees=1}
 {bone1Name=dyn_hair_ini_01, bone2Name=base_hair_ini_01, swing1LimitDegrees=2, swing2LimitDegrees=2, twistLimitDegrees=1}
 {bone1Name=dyn_hair_tip_01, bone2Name=dyn_hair_ini_01, swing1LimitDegrees=2, swing2LimitDegrees=2, twistLimitDegrees=1}
 {bone1Name=dyn_hair_ini_02, bone2Name=base_hair_ini_02, swing1LimitDegrees=3, swing2LimitDegrees=3, twistLimitDegrees=1}
 {bone1Name=dyn_hair_tip_02, bone2Name=dyn_hair_ini_02, swing1LimitDegrees=3, swing2LimitDegrees=3, twistLimitDegrees=1}
 {bone1Name=dyn_hair_ini_03, bone2Name=base_hair_ini_03, swing1LimitDegrees=6, swing2LimitDegrees=8, twistLimitDegrees=1}
 {bone1Name=dyn_hair_tip_03, bone2Name=dyn_hair_ini_03, swing1LimitDegrees=6, swing2LimitDegrees=8, twistLimitDegrees=1}
 {bone1Name=dyn_hair_ini_04, bone2Name=base_hair_ini_04, swing1LimitDegrees=6, swing2LimitDegrees=6, twistLimitDegrees=1}
 {bone1Name=dyn_hair_tip_04, bone2Name=dyn_hair_ini_04, swing1LimitDegrees=5, swing2LimitDegrees=5, twistLimitDegrees=1}
 {bone1Name=dyn_hair_ini_05, bone2Name=base_hair_ini_05, swing1LimitDegrees=5, swing2LimitDegrees=5, twistLimitDegrees=1}
 {bone1Name=dyn_hair_tip_05, bone2Name=dyn_hair_ini_05, swing1LimitDegrees=5, swing2LimitDegrees=5, twistLimitDegrees=1}
 {bone1Name=dyn_hair_ini_06, bone2Name=base_hair_ini_06, swing1LimitDegrees=5, swing2LimitDegrees=5, twistLimitDegrees=1}
 {bone1Name=dyn_hair_tip_06, bone2Name=dyn_hair_ini_06, swing1LimitDegrees=5, swing2LimitDegrees=5, twistLimitDegrees=1}
 {bone1Name=dyn_hair_ini_07, bone2Name=base_hair_ini_07, swing1LimitDegrees=2, swing2LimitDegrees=2, twistLimitDegrees=1}
 {bone1Name=dyn_hair_ini_08, bone2Name=base_hair_ini_08, swing1LimitDegrees=2, swing2LimitDegrees=2, twistLimitDegrees=1}
 {bone1Name=dyn_hoodie, bone2Name=base_hoodie, swing1LimitDegrees=8, swing2LimitDegrees=25, twistLimitDegrees=0}
 {bone1Name=dyn_B_hoodie, bone2Name=base_B_hoodie, swing1LimitDegrees=8, swing2LimitDegrees=8, twistLimitDegrees=0}
]
