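# Physics asset for the M_MED_PitGlass mask face accessory, a male medium player cosmetic, covering left, right and upper hood bones.
# Each hood section pairs a small capsule on a base bone with a larger, mostly offset capsule on a dyn bone.
# A constraint joins each pair with tight swing and twist limits, giving the upper section less swing than the sides.
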PhysicsAsset /BRCosmetics/Characters/Player/Male/Medium/Bodies/M_MED_PitGlass/Meshes/Parts/M_MED_PitGlass_Mask_FaceAcc_Physics
bodies {
 base_hood_l [{shapeType=Capsule, length=0.505, radius=0.505, shapeName=base_hood_l_capsule, rotation=<rot 0 -90 0>}]
 base_hood_r [{shapeType=Capsule, length=0.505, radius=0.505, shapeName=base_hood_r_capsule, rotation=<rot 0 -90 0>}]
 base_hood_up [{shapeType=Capsule, length=0.505, radius=0.505, shapeName=base_hood_up_capsule, rotation=<rot 0 -90 0>}]
 dyn_hood_l [{shapeType=Capsule, length=2, radius=1, shapeName=dyn_hood_l_capsule, center=<2 0 0>, rotation=<rot 90 0 0>}]
 dyn_hood_r [{shapeType=Capsule, length=2, radius=1, shapeName=dyn_hood_r_capsule, center=<1 0 0>, rotation=<rot 90 0 0>}]
 dyn_hood_up [{shapeType=Capsule, length=2, radius=1, shapeName=dyn_hood_up_capsule, center=<2 0 0>, rotation=<rot 90 0 0>}]
}
constraints [
 {bone1Name=dyn_hood_l, bone2Name=base_hood_l, swing1LimitDegrees=5, swing2LimitDegrees=15, twistLimitDegrees=1}
 {bone1Name=dyn_hood_r, bone2Name=base_hood_r, swing1LimitDegrees=5, swing2LimitDegrees=15, twistLimitDegrees=1}
 {bone1Name=dyn_hood_up, bone2Name=base_hood_up, swing1LimitDegrees=2, swing2LimitDegrees=5, twistLimitDegrees=1}
]
