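# Anim Blueprint for the PitGlass male mask face accessory that follows the body's pose and keeps the hood clear of the arms.
# It copies the attached parent's pose, offsets both hood bones, then runs head-relative RigidBody simulation with overridden gravity from a deprecated helper.
# Hood blend alphas come from clavicle-to-upperarm rotation on the cached body mesh.
# The skydive and parachute branches all set identical offset values.
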AnimBlueprint /BRCosmetics/Characters/Player/Male/Medium/Bodies/M_MED_PitGlass/Meshes/Parts/M_MED_PitGlass_Mask_FaceAcc_AnimBP
target_skeleton=&/M_MED_PitGlass_Mask_FaceAcc_Skeleton
graphs [
 AnimGraph {nodes [
   n1 = "Transform (Modify) Bone - Bone: base_hood_l"(
    ComponentPose=Local To Component(LocalPose=Copy Pose From Mesh() {bUseAttachedParent=true})
    Alpha=Get avoid_clipping_hoodie_with_arm_l() -> (avoid_clipping_hoodie_with_arm_l:double)
    Translation=Get avoid_clipping_hoddie_with_arm_l_parachute() -> (avoid_clipping_hoddie_with_arm_l_parachute:Vector)
    Rotation=Get avoid_clipping_hoddie_with_arm_l_parachute_Rotate() -> (avoid_clipping_hoddie_with_arm_l_parachute_Rotate:Rotator)
    Scale="1,1,1"
   ) {
    rotationMode=BMM_Replace
    rotationSpace=BCS_BoneSpace
    scaleMode=BMM_Additive
    scaleSpace=BCS_BoneSpace
    translationMode=BMM_Additive
    translationSpace=BCS_BoneSpace
    alphaBoolBlend {blendOption=Linear}
    alphaScaleBiasClamp {bInterpResult=true}
    boneToModify {boneName=base_hood_l}
    rotation {roll=-15, yaw=10}
    translation {y=5}
   }
   n2 = "Transform (Modify) Bone - Bone: base_hood_r"(
    ComponentPose=@n1
    Alpha=Get avoid_clipping_hoodie_with_arm_r() -> (avoid_clipping_hoodie_with_arm_r:double)
    Translation=Get avoid_clipping_hoddie_with_arm_r_parachute() -> (avoid_clipping_hoddie_with_arm_r_parachute:Vector)
    Rotation=Get avoid_clipping_hoddie_with_arm_r_parachute_Rotate() -> (avoid_clipping_hoddie_with_arm_r_parachute_Rotate:Rotator)
    Scale="1,1,1"
   ) {
    rotationMode=BMM_Replace
    rotationSpace=BCS_BoneSpace
    translationMode=BMM_Additive
    translationSpace=BCS_BoneSpace
    alphaBoolBlend {blendOption=Linear}
    alphaScaleBiasClamp {bInterpResult=true}
    boneToModify {boneName=base_hood_r}
    rotation {roll=-15, yaw=10}
    translation {y=-5}
   }
   n3 = RigidBody!(ComponentPose=@n2, Alpha, OverrideWorldGravity:Vector=Get GravityOverride(), ExternalForce) -> (Pose) {
    bOverrideWorldGravity=true
    overridePhysicsAsset=&/M_MED_PitGlass_Mask_FaceAcc_Physics
    simulationSpace=BaseBoneSpace
    alphaBoolBlend {blendOption=Linear}
    baseBoneRef {boneName=head}
    componentAppliedLinearAccClamp=<3000 3000 3000>
    componentLinearAccScale=<0.2 0.2 0.2>
    componentLinearVelScale=<0.2 0.2 0.2>
    simSpaceSettings {
     maxAngularAcceleration=1000
     maxAngularVelocity=3
     maxLinearAcceleration=700
     maxLinearVelocity=500
     velocityScaleZ=0
     worldAlpha=0.2
     externalLinearDragV=<0.2 0.2>
    }
   }
   Output Pose(Result=Component To Local(ComponentPose=@n3)) {layerGroup=None}
  ]}
 EventGraph {nodes [
   Event BlueprintUpdateAnimation()
   > n1 = Sequence!(execute) -> (then_0, then_1, then_2:Exec, then_3:Exec, then_4:Exec, then_5:Exec)
   n2 = Cast To PlayerPawn_Athena(execute=[@n7.then, @n1.then_2], Object=TryGetPawnOwner())
   > n3 = Set Body_Mesh(Body_Mesh=GetSkeletalMeshForPartType(self=@n2))
   n4 = Branch(execute=@n1.then_3, Condition=Get bIsSkydiveFloating())
   Set avoid_clipping_hoddie_with_arm_l_parachute(execute=[@n4.then, @n5.else], avoid_clipping_hoddie_with_arm_l_parachute:Vector="0,10,0") -> (Output_Get:Vector)
   > Set avoid_clipping_hoddie_with_arm_r_parachute(avoid_clipping_hoddie_with_arm_r_parachute:Vector="0,-10,0") -> (Output_Get:Vector)
   > Set avoid_clipping_hoddie_with_arm_l_parachute_Rotate(avoid_clipping_hoddie_with_arm_l_parachute_Rotate:Rotator="0,-40,40") -> (Output_Get:Rotator)
   > Set avoid_clipping_hoddie_with_arm_r_parachute_Rotate(avoid_clipping_hoddie_with_arm_r_parachute_Rotate:Rotator="0,35,-40") -> (Output_Get:Rotator)
   n5 = Branch(execute=@n4.else, Condition=Get bIsParachuteOpened())
   > Set avoid_clipping_hoddie_with_arm_l_parachute(avoid_clipping_hoddie_with_arm_l_parachute="0,10,0")
   > Set avoid_clipping_hoddie_with_arm_r_parachute(avoid_clipping_hoddie_with_arm_r_parachute="0,-10,0")
   > Set avoid_clipping_hoddie_with_arm_l_parachute_Rotate(avoid_clipping_hoddie_with_arm_l_parachute_Rotate="0,-40,40")
   > Set avoid_clipping_hoddie_with_arm_r_parachute_Rotate(avoid_clipping_hoddie_with_arm_r_parachute_Rotate="0,35,-40")
   n6 = Gravity Override State Function DEPRECATED(
    execute=@n1.then_0
    CharacterPartAnimInstance=Self-Reference()
    GameTimeInSeconds=GetGameTimeInSeconds()
    GravityOverrideParams=Get Gravity Override Params()
   )
   > Set GravityOverride(GravityOverride=@n6)
   Set avoid_clipping_hoodie_with_arm_l(
    execute=@n1.then_1
    avoid_clipping_hoodie_with_arm_l:double=GetAlphaFromReferenceTransformDelta(SkeletalMeshComponent=@n3, BoneName=clavicle_l, BaseName=upperarm_l, DeltaComponentType=EulerY, InputRange=(X=0,Y=-45), TwistAxis="0,0,1")
   ) -> (Output_Get:double)
   > Set avoid_clipping_hoodie_with_arm_r(avoid_clipping_hoodie_with_arm_r:double=GetAlphaFromReferenceTransformDelta(SkeletalMeshComponent=@n3, BoneName=clavicle_r, BaseName=upperarm_r, DeltaComponentType=EulerY, InputRange=(X=0,Y=-45), TwistAxis="0,0,1")) -> (Output_Get:double)
   n7 = Event OnCharacterCustomizationLoaded()
  ]}
]
variables [
 avoid_clipping_hoddie_with_arm_l_parachute {type=Vector}
 avoid_clipping_hoddie_with_arm_l_parachute_Rotate {type=Rotator}
 avoid_clipping_hoddie_with_arm_r_parachute {type=Vector}
 avoid_clipping_hoddie_with_arm_r_parachute_Rotate {type=Rotator}
 avoid_clipping_hoodie_with_arm_l {type=number}
 avoid_clipping_hoodie_with_arm_r {type=number}
 Body_Mesh {type=SkeletalMeshComponent}
 Gravity Override Params {type=GravityOverrideParamsStruct}
 GravityOverride {type=Vector}
]
